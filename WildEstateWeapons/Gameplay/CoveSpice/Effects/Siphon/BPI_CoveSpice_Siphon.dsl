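; Blueprint interface for the CoveSpice Siphon effect in the WildEstateWeapons gameplay content.
; It declares two parameterless functions, StartedHittingPawn and StoppedHittingPawn, that signal when the siphon begins and ceases hitting a pawn.
; It holds no logic; implementing objects supply the response to each event.
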(fn StartedHittingPawn)

(fn StoppedHittingPawn)
