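; Event-triggered gameplay ability for CoveSpice Siphon that heals its owner, on elimination judging by the name.
; After calling the parent, it builds the healing GameplayEffect spec, sets its SetByCaller.Healing magnitude and applies it to the owner.
; GetHealingMagnitude returns the scalable heal amount, or the gap to the configured max healed health if that would be reached.
; That gap has no zero floor, so it goes negative when health exceeds the cap.
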(event Ability|EventActivateAbilityFromEvent
  (params (EventData :type Gameplay%20Event%20Data%20Structure%20%28by%20ref%29))
  (|CallParentFunction EventData)
  (bind _returnvalue (Ability|GameplayEffect|AssignTagSetbyCallerMagnitude (Ability|MakeOutgoingGameplayEffectSpec (Variables|Default|GetHealing%20GE)) "(TagName=\"SetByCaller.Healing\")" (CallFunction|GetHealingMagnitude)))
  (Ability|ApplyGameplayEffectSpecToOwner _returnvalue))

(fn GetHealingMagnitude
  (return (Math|Float|SelectFloat (ScalableFloat|GetValueatLevel (Variables|Default|GetAmount%20to%20Heal)) (- (ScalableFloat|GetValueatLevel (Variables|Default|GetMax%20Healed%20Health%20Amount)) (FortPawn|GetHealth (CallFunction|GetAvatarasFortPlayerPawn))) (not (>= (+ (FortPawn|GetHealth (CallFunction|GetAvatarasFortPlayerPawn)) (ScalableFloat|GetValueatLevel (Variables|Default|GetAmount%20to%20Heal))) (ScalableFloat|GetValueatLevel (Variables|Default|GetMax%20Healed%20Health%20Amount)))))))
