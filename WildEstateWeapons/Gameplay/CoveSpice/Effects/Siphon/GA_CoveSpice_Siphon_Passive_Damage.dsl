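; Passive gameplay ability for the CoveSpice Siphon weapon that heals its owner on damage events against other targets matching the siphon filter tags.
; Event magnitude is clamped to scalable min/max limits, capped at missing health, and applied through a SetByCaller.Healing effect.
; Below the siphon health threshold, it sends the current weapon StartedHittingPawn; otherwise it removes any active exotic healing cue.
; Non-qualifying targets send the weapon StoppedHittingPawn instead.
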(event Ability|EventActivateAbilityFromEvent
  (params (EventData :type Gameplay%20Event%20Data%20Structure%20%28by%20ref%29))
  (bind (_eventtag _instigator _target) (Utilities|Struct|BreakGameplayEventData EventData))
  (bind _exotic_healing_gcnl (Variables|Default|GetExotic%20Healing%20GCNL))
  (|CallParentFunction EventData)
  (bind _outfortplayerpawn (CallFunction|GetAvatarasValidFortPlayerPawn)
    (:ValidFortPlayerPawn
      (bind _returnvalue (GameplayTags|HasAnyMatchingGameplayTags(Message) _target (GameplayTags|MakeLiteralGameplayTagContainer (Variables|Default|GetSiphon%20Filter%20Tags))))
      (if (and _returnvalue (not (== _outfortplayerpawn _target)))
        (bind _returnvalue_2 (Ability|GameplayEffect|AssignTagSetbyCallerMagnitude (Ability|MakeOutgoingGameplayEffectSpec (Variables|Default|GetSiphon%20Heal%20GE)) "(TagName=\"SetByCaller.Healing\")" (CallFunction|GetHealMagnitude EventData)))
        (Ability|ApplyGameplayEffectSpecToOwner _returnvalue_2)
        (if (< (FortPawn|GetHealth _outfortplayerpawn) (ScalableFloat|GetValueatLevel (Variables|Default|GetSiphon%20Max%20Health%20Amount)))
          (Class|BPICoveSpiceSiphon|StartedHittingPawn(Message) (Weapon|GetCurrentWeapon _outfortplayerpawn))
          (else
            ;[0.15 0.1434 0.000642 0.5] If healing GCNL is already active, remove it
            (bind _returnvalue_3 (Ability|GameplayCue|IsGameplayCueActive _outfortplayerpawn _exotic_healing_gcnl))
            (if _returnvalue_3
              (GameplayCue|RemoveGameplayCueOnActor(Looping) _outfortplayerpawn _exotic_healing_gcnl (Ability|GameplayCue|MakeGameplayCueParameters :NormalizedMagnitude 0.0 :RawMagnitude 0.0 :Location "0,0,0" :Normal "0,0,0" :Instigator _outfortplayerpawn :GameplayEffectLevel 1 :AbilityLevel 1 :bReplicateLocationWhenUsingMinimalRepProxy false)))))
        (else
          (Class|BPICoveSpiceSiphon|StoppedHittingPawn(Message) (Weapon|GetCurrentWeapon _outfortplayerpawn)))))
    (:AvatarCastFailed
      (Ability|EndAbility))))

(fn GetClampedHealMagnitude
  (params (Value :type Float%20%28double-precision%29))
  (return (Math|Float|Clamp(Float) Value (ScalableFloat|GetValueatLevel (Variables|Default|GetSiphon%20Heal%20Min)) (ScalableFloat|GetValueatLevel (Variables|Default|GetSiphon%20Heal%20Max)))))

(fn GetHealMagnitude
  (params (GameplayEventData :type Gameplay%20Event%20Data%20Structure%20%28by%20ref%29))
  (bind (_eventtag _instigator _target _optionalobject _optionalobject2 _contexthandle _instigatortags _targettags _eventmagnitude) (Utilities|Struct|BreakGameplayEventData (Variables|GetGameplayEventData)))
  (bind _player_s_max_health (LocalVariables|GetPlayer's%20Max%20Health))
  (Utilities|FlowControl|Sequence
    (:then_0
      ;[0 0.15 0.00729 0.5] Cache used values
      (LocalVariables|SetHeal%20Amount (CallFunction|GetClampedHealMagnitude _eventmagnitude))
      (LocalVariables|SetPlayer's%20Current%20Health (FortPawn|GetHealth (CallFunction|GetAvatarasFortPlayerPawn)))
      (LocalVariables|SetPlayer's%20Max%20Health (FortPawn|GetMaxHealth (CallFunction|GetAvatarasFortPlayerPawn))))
    (:then_1
      ;[0.1336 0 0.15 0.5] Calculate Amount to Heal
      (if (>= (+ (LocalVariables|GetHeal%20Amount) (LocalVariables|GetPlayer's%20Current%20Health)) _player_s_max_health)
        (LocalVariables|SetHeal%20Amount (Math|Float|Clamp(Float) (- _player_s_max_health (LocalVariables|GetPlayer's%20Current%20Health)) 0.0 _player_s_max_health))))
    (:then_2
      (return (LocalVariables|GetHeal%20Amount)))))
