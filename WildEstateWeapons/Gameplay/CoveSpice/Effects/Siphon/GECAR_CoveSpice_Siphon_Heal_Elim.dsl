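; Custom application requirement for the CoveSpice Siphon heal effect, apparently the one granted on eliminations, that gates whether the heal can apply.
; It casts the target ability system component's owner to a Fort player state and reads the current pawn's health.
; The effect is allowed only while that health is below the scalable Max Player Health to Heal to value.
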(fn CanApplyGameplayEffect
  (params (GameplayEffect :type Gameplay%20Effect%20Object%20Reference) (Spec :type Gameplay%20Effect%20Spec%20Structure%20%28by%20ref%29) (ASC :type Ability%20System%20Component%20Object%20Reference))
  (bind _asfort_player_state (Utilities|Casting|CastToFortPlayerState (Components|GetOwner ASC)))
  (return (< (FortPawn|GetHealth (PlayerState|GetCurrentPawn _asfort_player_state)) (ScalableFloat|GetValueatLevel (Variables|Default|GetMax%20Player%20Health%20to%20Heal%20to)))))
